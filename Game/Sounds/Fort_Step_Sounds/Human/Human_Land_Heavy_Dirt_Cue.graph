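# Sound cue for a human player's heavy landing on dirt, routed to the player foley sound class.
# A mixer layers three branches: a random jump-land sample with slight downward pitch variation, a randomly chosen dirt run step after a short random delay, and a single dirt land sample, also briefly delayed and mixed at lower volume.
# The cue overrides attenuation and concurrency, preventing new instances once the limit is reached.
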SoundCue /Game/Sounds/Fort_Step_Sounds/Human/Human_Land_Heavy_Dirt_Cue
root=Mixer_2
nodes [
 {id=Delay_2, children [Modulator_5], properties {delayMax=0.15, delayMin=0.085}}
 {id=Delay_4, children [Modulator_9], properties {delayMax=0.1, delayMin=0.05}}
 {id=Mixer_2, children [Modulator_4,Delay_2,Delay_4], properties {inputVolume [1,1,0.6]}}
 {id=Modulator_4, children [Random_0], properties {pitchMax=1, pitchMin=0.9}}
 {id=Modulator_5, children [Random_7]}
 {id=Modulator_9, children [WavePlayer_304]}
 {id=Random_0, children [WavePlayer_0,WavePlayer_1,WavePlayer_2,WavePlayer_3], properties {weights [1,1,1,1]}}
 {
  id=Random_7
  children [WavePlayer_307, WavePlayer_306, WavePlayer_305, WavePlayer_302, WavePlayer_308, WavePlayer_309]
  properties {weights [1,1,1,1,1,1]}
 }
 {id=WavePlayer_0, properties {soundWaveAssetPtr=&/Human_Jump_Land01}}
 {id=WavePlayer_1, properties {soundWaveAssetPtr=&/Human_Jump_Land02}}
 {id=WavePlayer_2, properties {soundWaveAssetPtr=&/Human_Jump_Land03}}
 {id=WavePlayer_3, properties {soundWaveAssetPtr=&/Human_Jump_Land04}}
 {id=WavePlayer_302, properties {soundWaveAssetPtr=&/player_step_run_dirt_04}}
 {id=WavePlayer_304, properties {soundWaveAssetPtr=&/player_step_land_dirt_01}}
 {id=WavePlayer_305, properties {soundWaveAssetPtr=&/player_step_run_dirt_03}}
 {id=WavePlayer_306, properties {soundWaveAssetPtr=&/player_step_run_dirt_02}}
 {id=WavePlayer_307, properties {soundWaveAssetPtr=&/player_step_run_dirt_01}}
 {id=WavePlayer_308, properties {soundWaveAssetPtr=&/player_step_run_dirt_05}}
 {id=WavePlayer_309, properties {soundWaveAssetPtr=&/player_step_run_dirt_06}}
]
properties {
 bOverrideAttenuation=true
 bOverrideConcurrency=true
 maxDistance=4000
 priority=0.5
 soundClassObject=&Fort_Player_Foley
 concurrencyOverrides {resolutionRule=PreventNew}
}
