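# Sound cue for a human player's light landing on dirt, played as player foley at reduced volume.
# A mixer layers a random jump-land sample with a random dirt walk-step sample that follows after a short randomized delay.
# Each layer passes through its own modulator for variation.
# Attenuation and concurrency are overridden on the cue, with new instances blocked once the concurrency limit is reached.
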SoundCue /Game/Sounds/Fort_Step_Sounds/Human/Human_Land_Light_Dirt_Cue
root=Mixer_0
nodes [
 {id=Delay_0, children [Modulator_1], properties {delayMax=0.15, delayMin=0.075}}
 {id=Mixer_0, children [Modulator_3,Delay_0], properties {inputVolume [1,1]}}
 {id=Modulator_1, children [Random_4]}
 {id=Modulator_3, children [Random_1]}
 {id=Random_1, children [WavePlayer_4,WavePlayer_5,WavePlayer_6,WavePlayer_7], properties {weights [1,1,1,1]}}
 {id=Random_4, children [WavePlayer_310, WavePlayer_312, WavePlayer_313, WavePlayer_314, WavePlayer_315], properties {weights [1,1,1,1,1]}}
 {id=WavePlayer_310, properties {soundWaveAssetPtr=&/player_step_walk_dirt_01}}
 {id=WavePlayer_312, properties {soundWaveAssetPtr=&/player_step_walk_dirt_02}}
 {id=WavePlayer_313, properties {soundWaveAssetPtr=&/player_step_walk_dirt_03}}
 {id=WavePlayer_314, properties {soundWaveAssetPtr=&/player_step_walk_dirt_04}}
 {id=WavePlayer_315, properties {soundWaveAssetPtr=&/player_step_walk_dirt_05}}
 {id=WavePlayer_4, properties {soundWaveAssetPtr=&/Human_Jump_Land01}}
 {id=WavePlayer_5, properties {soundWaveAssetPtr=&/Human_Jump_Land02}}
 {id=WavePlayer_6, properties {soundWaveAssetPtr=&/Human_Jump_Land03}}
 {id=WavePlayer_7, properties {soundWaveAssetPtr=&/Human_Jump_Land04}}
]
properties {
 bOverrideAttenuation=true
 bOverrideConcurrency=true
 maxDistance=4000
 priority=0.5
 soundClassObject=&Fort_Player_Foley
 volumeMultiplier=0.5
 concurrencyOverrides {resolutionRule=PreventNew}
}
